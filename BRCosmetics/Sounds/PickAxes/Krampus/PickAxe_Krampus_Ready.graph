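# Sound cue for the ready sound of the Krampus pickaxe cosmetic, built from swing waves.
# A quality-level node picks the branch, with random variations shrinking from six close swings to three to a single wave shared by the last two tiers.
# Each branch mixes Krampus swings under close attenuation with quieter, reused distant fire-axe swings under medium attenuation.
# An owner-limited concurrency override stops the oldest instance.
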SoundCue /BRCosmetics/Sounds/PickAxes/Krampus/PickAxe_Krampus_Ready
root=QualityLevel_0
nodes [
 {id=Attenuation_0, children [WavePlayer_18], properties {attenuationSettings=/Game/Sounds/Attenuation/Fort_Weapon_Pickaxe_Close_Att}}
 {id=Attenuation_2, children [Random_0], properties {attenuationSettings=&Fort_Weapon_Pickaxe_Close_Att}}
 {id=Attenuation_4, children [Random_5], properties {attenuationSettings=&Fort_Weapon_Pickaxe_Close_Att}}
 {id=Attenuation_7, children [Modulator_4], properties {attenuationSettings=/Game/Sounds/Attenuation/Fort_Weapon_Pickaxe_Medium_Att}}
 {id=Attenuation_8, children [Random_6], properties {attenuationSettings=&Fort_Weapon_Pickaxe_Medium_Att}}
 {id=Attenuation_9, children [WavePlayer_44], properties {attenuationSettings=&Fort_Weapon_Pickaxe_Medium_Att}}
 {id=Mixer_0, children [Attenuation_0, Modulator_8], properties {inputVolume [1,1]}}
 {id=Mixer_1, children [Attenuation_2, Attenuation_7], properties {inputVolume [1.1,1]}}
 {id=Mixer_2, children [Attenuation_4, Modulator_7], properties {inputVolume [1.1,1]}}
 {id=Modulator_1, children [Mixer_1], properties {pitchMax=1.1, pitchMin=1, volumeMax=1}}
 {id=Modulator_2, children [Mixer_2]}
 {id=Modulator_4, children [Random_1], properties {pitchMax=1.1, pitchMin=1, volumeMax=0.35, volumeMin=0.35}}
 {id=Modulator_7, children [Attenuation_8], properties {pitchMax=1.1, pitchMin=1, volumeMax=0.35, volumeMin=0.35}}
 {id=Modulator_8, children [Attenuation_9], properties {pitchMax=1.1, pitchMin=1, volumeMax=0.35, volumeMin=0.35}}
 {id=QualityLevel_0, children [Modulator_1,Modulator_2,Mixer_0,Mixer_0]}
 {
  id=Random_0
  children [WavePlayer_1, WavePlayer_2, WavePlayer_4, WavePlayer_5, WavePlayer_12, WavePlayer_13]
  properties {weights [1,1,1,1,1,1]}
 }
 {
  id=Random_1
  children [WavePlayer_19, WavePlayer_20, WavePlayer_23, WavePlayer_26, WavePlayer_27, WavePlayer_30]
  properties {weights [1,1,1,1,1,1]}
 }
 {id=Random_5, children [WavePlayer_15, WavePlayer_16, WavePlayer_17], properties {weights [1,1,1]}}
 {
  id=Random_6
  children [WavePlayer_43, WavePlayer_42, WavePlayer_41, WavePlayer_31, WavePlayer_38, WavePlayer_40]
  properties {weights [1,1,1,1,1,1]}
 }
 {id=WavePlayer_1, properties {soundWaveAssetPtr=&/PA_Krampus_Swing_01}}
 {id=WavePlayer_12, properties {soundWaveAssetPtr=&/PA_Krampus_Swing_05}}
 {id=WavePlayer_13, properties {soundWaveAssetPtr=&/PA_Krampus_Swing_06}}
 {id=WavePlayer_15, properties {soundWaveAssetPtr=&PA_Krampus_Swing_01}}
 {id=WavePlayer_16, properties {soundWaveAssetPtr=&/PA_Krampus_Swing_02}}
 {id=WavePlayer_17, properties {soundWaveAssetPtr=&/PA_Krampus_Swing_03}}
 {id=WavePlayer_18, properties {soundWaveAssetPtr=&PA_Krampus_Swing_01}}
 {id=WavePlayer_19, properties {soundWaveAssetPtr=/Game/Athena/Sounds/Weapons/Fire_Axe_Swing_Distant_04}}
 {id=WavePlayer_2, properties {soundWaveAssetPtr=&PA_Krampus_Swing_02}}
 {id=WavePlayer_20, properties {soundWaveAssetPtr=/Game/Athena/Sounds/Weapons/Fire_Axe_Swing_Distant_01}}
 {id=WavePlayer_23, properties {soundWaveAssetPtr=/Game/Athena/Sounds/Weapons/Fire_Axe_Swing_Distant_02}}
 {id=WavePlayer_26, properties {soundWaveAssetPtr=/Game/Athena/Sounds/Weapons/Fire_Axe_Swing_Distant_03}}
 {id=WavePlayer_27, properties {soundWaveAssetPtr=/Game/Athena/Sounds/Weapons/Fire_Axe_Swing_Distant_05}}
 {id=WavePlayer_30, properties {soundWaveAssetPtr=/Game/Athena/Sounds/Weapons/Fire_Axe_Swing_Distant_06}}
 {id=WavePlayer_31, properties {soundWaveAssetPtr=&Fire_Axe_Swing_Distant_03}}
 {id=WavePlayer_38, properties {soundWaveAssetPtr=&Fire_Axe_Swing_Distant_05}}
 {id=WavePlayer_4, properties {soundWaveAssetPtr=&PA_Krampus_Swing_03}}
 {id=WavePlayer_40, properties {soundWaveAssetPtr=&Fire_Axe_Swing_Distant_06}}
 {id=WavePlayer_41, properties {soundWaveAssetPtr=&Fire_Axe_Swing_Distant_02}}
 {id=WavePlayer_42, properties {soundWaveAssetPtr=&Fire_Axe_Swing_Distant_01}}
 {id=WavePlayer_43, properties {soundWaveAssetPtr=&Fire_Axe_Swing_Distant_04}}
 {id=WavePlayer_44, properties {soundWaveAssetPtr=&Fire_Axe_Swing_Distant_01}}
 {id=WavePlayer_5, properties {soundWaveAssetPtr=&/PA_Krampus_Swing_04}}
]
properties {
 audioPropertiesSheet=/BRCosmetics/Sounds/Settings/PropertySheets/Pickaxe/APS_Cosmetic_Pickaxe_Ready
 bOverrideConcurrency=true
 bPrimeOnLoad=true
 maxDistance=3000
 pitchMultiplier=1.1
 soundClassObject=&Fort_WeaponMelee
 volumeMultiplier=0.15
 concurrencyOverrides {bLimitToOwner=true, maxCount=2, resolutionRule=StopOldest}
 concurrencySet [&Player_Melee_Weapon_Conc]
}
